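; Parent passive ability for Season 22 Battle Pass Chromehands TV quests that plays a montage on the player with the weapon holstered.
; A repeating one-second timer stops the montage if the player moves from the saved start location beyond a tolerance.
; After a short delay, a primary-press gameplay event also stops it.
; Ending the ability clears the timer, releases the holster and removes gameplay effects matching a removal tag.
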(event Ability|EventActivateAbility
  (|CallParentFunction)
  (Ability|Tasks|PlayMontageInfoAndWaitForTrigger "None" (Variables|Animation|GetMontageInfo)
    (:then
      (Weapon|SetHolsterWeapon (Variables|GAVars|Data|GetOwningPlayerPawn) 0 true false (Variables|Default|GetBitesHolsterID))
      (Variables|Default|SetPlayerLoc (Transformation|GetActorLocation (Variables|GAVars|Data|GetOwningPlayerPawn)))
      (bind _returnvalue (Utilities|Time|SetTimerbyEvent (delegate Custom|check%20if%20player%20moved) 1.0 true))
      (Variables|Default|SetTimerHandle _returnvalue)
      (Utilities|FlowControl|Delay 2.0)
      (Ability|Tasks|WaitGameplayEvent "(TagName=\"Event.Primary.Pressed\")" (Variables|GAVars|Data|GetOwningPlayerPawn) true
        (:EventReceived
          (Ability|Animation|MontageStop -1.0)
          (Utilities|Time|ClearandInvalidateTimerbyHandle (Variables|Default|GetTimerHandle)))))
    (:Cancelled
      _)
    (:Completed
      _))
  (Ability|EndAbility))

(event Custom|check%20if%20player%20moved
  (if (not (Utilities|Operators|Equal(==) (Transformation|GetActorLocation (Variables|GAVars|Data|GetOwningPlayerPawn)) (Variables|Default|GetPlayerLoc) 5e+01))
    (Ability|Animation|MontageStop -1.0)
    (Utilities|Time|ClearandInvalidateTimerbyHandle (Variables|Default|GetTimerHandle))))

(event Ability|EventFailedToActivatePassiveAbility
  (params (ActorInfo :type Gameplay%20Ability%20Actor%20Info%20Structure))
  (|CallParentFunction ActorInfo))

(event Ability|EventOnEndAbility
  (params (bWasCancelled :type Boolean))
  (|CallParentFunction bWasCancelled)
  (Utilities|Time|ClearandInvalidateTimerbyHandle (Variables|Default|GetTimerHandle))
  (Weapon|SetHolsterWeapon (Variables|GAVars|Data|GetOwningPlayerPawn) 0 false false (Variables|Default|GetBitesHolsterID))
  (GameplayEffects|RemoveActiveEffectswithTags (Ability|GetAbilitySystemComponent (Variables|GAVars|Data|GetOwningPlayerPawn)) (Variables|Default|GetGE%20Removal%20Tag)))
